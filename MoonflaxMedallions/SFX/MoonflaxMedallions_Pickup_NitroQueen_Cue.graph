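# Sound cue for picking up the Nitro Queen medallion in Moonflax Medallions, layering a stinger over the default Season 28 takedown medallion pickup sound.
# A mixer combines the shared pickup sound at reduced volume with the character stinger, which plays at full level after a short delay.
# It routes through the loot-collect sound class and overrides concurrency to allow only one instance at a time.
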SoundCue /MoonflaxMedallions/SFX/MoonflaxMedallions_Pickup_NitroQueen_Cue
root=Mixer_2
nodes [
 {id=Delay_0, children [WavePlayer_0], properties {delayMax=0.2, delayMin=0.2}}
 {id=Mixer_2, children [WavePlayer_5,Delay_0], properties {inputVolume [0.6,1]}}
 {id=WavePlayer_0, properties {soundWaveAssetPtr=&/MX_MoonflaxMedallions_NitroQueen_Stinger}}
 {id=WavePlayer_5, properties {soundWaveAssetPtr=/TakedownMedallion/Audio/Medallions/S28/Default_TakedownMedallions_Pickup_SFX}}
]
properties {
 bOverrideConcurrency=true
 maxDistance=2097152
 soundClassObject=&Fort_Loot_Collect
 volumeMultiplier=0.7692
 concurrencyOverrides {maxCount=1}
}
